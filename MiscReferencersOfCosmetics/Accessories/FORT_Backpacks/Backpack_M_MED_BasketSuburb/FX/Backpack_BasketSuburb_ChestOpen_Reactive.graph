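# Level Sequence that plays chest-open reactive animation and FX for the BasketSuburb backpack cosmetic, alongside a second DBNO-named backpack animation.
# It possesses two skeletal mesh actors, each held at a static transform with one skeletal animation section, both starting at frame zero.
# The chest-burst Niagara component is bound under the chest-reactive actor and driven by a system spawn section that outlasts both animations.
# It has no root-level tracks.
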LevelSequence /MiscReferencersOfCosmetics/Accessories/FORT_Backpacks/Backpack_M_MED_BasketSuburb/FX/Backpack_BasketSuburb_ChestOpen_Reactive
bindings [
 NS_Backpack_BasketSuburb_Chest_Burst {
  id=#1
  parent_id=#2
  possessed_class=NiagaraComponent
  tracks [{type=MovieSceneNiagaraSystemTrack, sections [{type=MovieSceneNiagaraSystemSpawnSection, end_frame=300, start_frame=0}]}]
 }
 BasketSuburb_ChestReactive_PACK {
  id=#2
  possessed_class=FortSkeletalMeshActor
  tracks [
   {
    type=MovieScene3DTransformTrack
    sections [{
      type=MovieScene3DTransformSection
      channels [
       Location.X {type=MovieSceneScriptingDoubleChannel, default=-72}
       Location.Y {type=MovieSceneScriptingDoubleChannel, default=1480}
       Location.Z {type=MovieSceneScriptingDoubleChannel, default=1996}
       Rotation.X {type=MovieSceneScriptingDoubleChannel, default=0}
       Rotation.Y {type=MovieSceneScriptingDoubleChannel, default=0}
       Rotation.Z {type=MovieSceneScriptingDoubleChannel, default=0}
       Scale.X {type=MovieSceneScriptingDoubleChannel, default=1}
       Scale.Y {type=MovieSceneScriptingDoubleChannel, default=1}
       Scale.Z {type=MovieSceneScriptingDoubleChannel, default=1}
      ]
     }]
   }
   {type=MovieSceneSkeletalAnimationTrack, sections [{type=MovieSceneSkeletalAnimationSection, end_frame=63, start_frame=0}]}
  ]
 }
 BasketSuburb_DBNO_PACK {
  id=#3
  possessed_class=FortSkeletalMeshActor
  tracks [
   {
    type=MovieScene3DTransformTrack
    sections [{
      type=MovieScene3DTransformSection
      channels [
       Location.X {type=MovieSceneScriptingDoubleChannel, default=236}
       Location.Y {type=MovieSceneScriptingDoubleChannel, default=1600}
       Location.Z {type=MovieSceneScriptingDoubleChannel, default=1960}
       Rotation.X {type=MovieSceneScriptingDoubleChannel, default=0}
       Rotation.Y {type=MovieSceneScriptingDoubleChannel, default=0}
       Rotation.Z {type=MovieSceneScriptingDoubleChannel, default=0}
       Scale.X {type=MovieSceneScriptingDoubleChannel, default=1}
       Scale.Y {type=MovieSceneScriptingDoubleChannel, default=1}
       Scale.Z {type=MovieSceneScriptingDoubleChannel, default=1}
      ]
     }]
   }
   {type=MovieSceneSkeletalAnimationTrack, sections [{type=MovieSceneSkeletalAnimationSection, end_frame=48, start_frame=0}]}
  ]
 }
]
display_rate {denominator=1, numerator=30}
tick_resolution {denominator=1, numerator=24000}
tracks []
